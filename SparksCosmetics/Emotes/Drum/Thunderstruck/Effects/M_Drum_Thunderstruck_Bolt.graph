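# Unlit Niagara mesh-particle material for the lightning bolts in the Thunderstruck drum emote.
# World position offset adds squared scrolling noise, masked by vertex red, to a time-stepped rotation about local Z for jagged, jumping motion.
# A threshold on that rotation phase intermittently shoves vertex-green geometry far down Z, apparently hiding a second bolt.
# Emissive color runs through MF_FX_EmissiveAdjustments and flickers with a dark color using a fast time-based step.
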Material /SparksCosmetics/Emotes/Drum/Thunderstruck/Effects/M_Drum_Thunderstruck_Bolt
graph {
 properties {bAutomaticallySetUsageInEditor=false, bUsedWithNiagaraMeshParticles=true, shadingModel=MSM_Unlit, usedShadingModels=Unlit}
 nodes [
  n1 = Multiply(A=Multiply(A=NamedRerouteUsage(), B=VertexColor().R), B=Constant3Vector() {constant=<rgba 1 1 0.2 0>})
  n2 = ScalarParameter() {defaultValue=10, parameterName=WPO_Strength}
  n3 = ScalarParameter() {defaultValue=-1, parameterName=RotationSpeed}
  n4 = Time()
  n5 = Multiply(A=@n4, B=ScalarParameter() {defaultValue=2, parameterName=Speed})
  n6 = Ceil(None=Multiply(A=Frac(None=@n5), B=ScalarParameter() {defaultValue=5, parameterName=Steps}) {constB=7})
  n7 = ScalarParameter() {defaultValue=0.35, parameterName=Cut}
  n8 = Frac(None=Add(A=Multiply(A=@n3, B=@n4), B=Multiply(A=@n6, B=@n7)))
  n9 = RotateAboutAxis(
   NormalizedRotationAxis=Normalize(VectorInput=Constant3Vector() {constant=<rgba 0 0 1 0>})
   RotationAngle=@n8
   PivotPoint=Constant()
   Position=LocalPosition() {includedOffsets=ExcludeOffsets}.XYZ
  )
  n10 = Constant3Vector() {constant=<rgba 0 0 -10000 0>}
  n11 = VertexColor()
  n12 = ScalarParameter() {defaultValue=0.6, parameterName=2ndBolt_Likeliness_0-1}
  Add(
   A=Add(A=Multiply(A=@n1, B=@n2), B=Transform(None=@n9) {transformSourceType=TRANSFORMSOURCE_Local})
   B=LinearInterpolate(A=Multiply(A=@n10, B=@n11.G), Alpha=Step(Y=@n8, X=@n12)) {constAlpha=0, constB=0}
  )
  n13 = Time()
  n14 = Add(A=Multiply(A=ScalarParameter() {defaultValue=2.77, parameterName=ColorVariant_UVScale}, B=@n13), B=Multiply(A=@n13, B=Constant() {r=7.77}))
  LinearInterpolate(
   A=VectorParameter() {parameterName=Color2, defaultValue=<rgba 0 0 0.0001 0>}.RGB
   B=Multiply(B=NamedRerouteUsage())
   Alpha=Step(Y=Frac(None=@n14), X=ScalarParameter() {defaultValue=0.98, parameterName=DarkBolt_Likeliness_0-1}) {constX=0.85}
  )
  MaterialFunctionCall(Emissive=Multiply(
    A=VectorParameter() {parameterName=Color, defaultValue=<rgba 0 0.2 1 1>}.RGB
    B=ScalarParameter() {defaultValue=10, parameterName=Brightness}
   )) {materialFunction=&MF_FX_EmissiveAdjustments}
  n15 = TextureCoordinate()
  n16 = VectorParameter() {parameterName=Noise_Scale-Pan, defaultValue=<rgba 0 2 0 -2>}
  n17 = AppendVector(A=@n16.R, B=@n16.R)
  n18 = TextureSample(UVs=Panner(Coordinate=Multiply(A=@n15, B=@n17), Speed=AppendVector(A=@n16.B, B=@n16.B))) {samplerType=SAMPLERTYPE_Grayscale, texture=&LowResBlurredNoise}
  NamedRerouteDeclaration(None=Multiply(A=@n18.R, B=@n18.R)) {name=ScrollNoise, nodeColor=<rgba 0 1 0.9063 1>}
 ]
}
