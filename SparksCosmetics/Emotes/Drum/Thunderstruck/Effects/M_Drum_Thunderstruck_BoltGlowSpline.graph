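# Translucent unlit glow material for lightning bolt splines in the Thunderstruck drum emote effects, built for Niagara mesh particles.
# World position offset combines SplineThicken width, vertex-red-weighted displacement, and a Z-axis rotation that spins steadily with repeating stepped jumps for flickering.
# Named reroutes supply squared scrolling noise and a lengthwise UV falloff to modulate the shape.
# A tinted emissive passes through MF_FX_EmissiveAdjustments, while a widthwise UV falloff sets soft-edged opacity.
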Material /SparksCosmetics/Emotes/Drum/Thunderstruck/Effects/M_Drum_Thunderstruck_BoltGlowSpline
graph {
 properties {
  bAutomaticallySetUsageInEditor=false
  bCastRayTracedShadows=false
  bUsedWithNiagaraMeshParticles=true
  blendMode=BLEND_Translucent
  shadingModel=MSM_Unlit
  twoSided=true
  usedShadingModels=Unlit
 }
 nodes [
  n1 = Multiply(A=ScalarParameter() {defaultValue=60, parameterName=WPOStrength}, B=NamedRerouteUsage())
  n2 = NamedRerouteUsage()
  n3 = VertexColor()
  n4 = Multiply(
   A=Multiply(A=Multiply(A=@n2, B=@n3.R), B=Constant3Vector() {constant=<rgba 1 1 0.2 0>})
   B=ScalarParameter() {defaultValue=10, parameterName=WPO_Strength}
  )
  n5 = Constant3Vector() {constant=<rgba 0 0 1 0>}
  n6 = Time()
  n7 = ScalarParameter() {defaultValue=2, parameterName=Speed}
  n8 = Multiply(A=Frac(None=Multiply(A=@n6, B=@n7)), B=ScalarParameter() {defaultValue=5, parameterName=Steps}) {constB=7}
  n9 = Add(
   A=Multiply(A=ScalarParameter() {defaultValue=-1, parameterName=RotationSpeed}, B=@n6)
   B=Multiply(A=Ceil(None=@n8), B=ScalarParameter() {defaultValue=0.35, parameterName=Cut})
  )
  n10 = Transform(None=RotateAboutAxis(
    NormalizedRotationAxis=Normalize(VectorInput=@n5)
    RotationAngle=Frac(None=@n9)
    PivotPoint=Constant()
    Position=LocalPosition() {includedOffsets=ExcludeOffsets}.XYZ
   )) {transformSourceType=TRANSFORMSOURCE_Local}
  Add(
   A=MaterialFunctionCall(WidthBase=@n1, WidthTip=@n1, Expand U or V UV Channel=StaticBool()) {materialFunction=&SplineThicken}.WorldPositionOffset
   B=Multiply(A=Add(A=@n4, B=@n10), B=ScalarParameter() {defaultValue=1, parameterName=Flippitude})
  )
  n11 = VectorParameter() {parameterName=Color, defaultValue=<rgba 0 0.2 1 1>}
  MaterialFunctionCall(Emissive=Multiply(A=Add(A=@n11.RGB) {constB=0.003}, B=ScalarParameter() {defaultValue=1.5, parameterName=Brightness})) {materialFunction=&MF_FX_EmissiveAdjustments}
  n12 = ComponentMask(None=TextureCoordinate()) {g=true}
  n13 = Multiply(A=@n12) {constB=2}
  n14 = Multiply(A=OneMinus(None=@n12)) {constB=2}
  n15 = Multiply(A=Multiply(A=Saturate(None=@n13), B=Saturate(None=@n14)), B=ScalarParameter() {defaultValue=1, parameterName=WidthCoreSize})
  n16 = ScalarParameter() {defaultValue=3, parameterName=WidthFalloffPow}
  Multiply(A=Saturate(None=Power(Base=@n15, Exp=@n16)), B=ScalarParameter() {defaultValue=0.5, parameterName=OpacityMaster})
  n17 = TextureCoordinate()
  n18 = VectorParameter() {parameterName=Noise_Scale-Pan, defaultValue=<rgba 0 2 0 -2>}
  n19 = AppendVector(A=@n18.R, B=@n18.R)
  n20 = TextureSample(UVs=Panner(Coordinate=Multiply(A=@n17, B=@n19), Speed=AppendVector(A=@n18.B, B=@n18.B))) {samplerType=SAMPLERTYPE_Grayscale, texture=&LowResBlurredNoise}
  NamedRerouteDeclaration(None=Multiply(A=@n20.R, B=@n20.R)) {name=ScrollNoise, nodeColor=<rgba 0 1 0.9062 1>}
  n21 = ComponentMask(None=TextureCoordinate()) {r=true}
  n22 = Multiply(A=@n21) {constB=2}
  n23 = Multiply(A=OneMinus(None=@n21)) {constB=2}
  n24 = Multiply(A=Multiply(A=Saturate(None=@n22), B=Saturate(None=@n23)), B=ScalarParameter() {defaultValue=1, parameterName=LengthCoreSize})
  n25 = ScalarParameter() {defaultValue=0.5, parameterName=LengthFalloffPow}
  NamedRerouteDeclaration(None=Saturate(None=Power(Base=@n24, Exp=@n25))) {name=LengthMask, nodeColor=<rgba 0.4531 0 1 1>}
 ]
}
